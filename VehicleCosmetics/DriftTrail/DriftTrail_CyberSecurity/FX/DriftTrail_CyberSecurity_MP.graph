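# Niagara system for the CyberSecurity vehicle drift trail cosmetic, emitting from wheel sockets on the car's skeletal mesh.
# Combines rear-wheel wrap meshes, twin ribbons and two triangle sprite emitters, with spawn groups merging two particle looks into the embers emitter.
# Per-wheel masks from FN_DisableTireFX gate each layer, while user parameters drive paint colors, outlines and front-end preview variants.
# A second wheel-wrap emitter is present but disabled.
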NiagaraSystem /VehicleCosmetics/DriftTrail/DriftTrail_CyberSecurity/FX/DriftTrail_CyberSecurity_MP
dependencies {
 usedDataInterfaces [
  &NiagaraDataInterfaceArrayFloat2
  &NiagaraDataInterfaceCurve
  &NiagaraDataInterfaceSpriteRendererInfo
  &NiagaraDataInterfaceSkeletalMesh
 ]
 usedDynamicInputs []
 usedModules [
  &self:SystemSpawnScript.SystemScriptSource.SystemScriptGraph.NiagaraNodeAssignment_0.SetVariables_#1
  &SystemState
  &FN_UnpackVehicleFlags
  /Game/Effects/Niagara/Modules/Vehicle/FN_UpdateVehicleAttribs
  /Game/Effects/Niagara/Modules/Vehicle/FN_VehicleFlags_Cheatcodes
 ]
}
emitters {
 Embers_Triangles {
  emitterUpdateScript [
   &EmitterState()
   &SpawnRate(
    SpawnRate=/VehicleCosmetics/FX/Niagara/Modules/VC_Mask_Floats(
     FrontEnd Value:NiagaraFloat=50
     blsFrontEndPreview:NiagaraBool=$User.bIsFrontendPreview
     SpeedMin Value:NiagaraFloat=30
     SpeedMax Value:NiagaraFloat=30
     Speed Alpha:NiagaraFloat=$System.Speed_Alpha
    )
    Use Spawn Probability=false
   )
   &SpawnRate(
    SpawnGroup=1
    SpawnRate=&VC_Mask_Floats(
     FrontEnd Value:NiagaraFloat=60
     blsFrontEndPreview:NiagaraBool=$User.bIsFrontendPreview
     SpeedMin Value:NiagaraFloat=60
     SpeedMax Value:NiagaraFloat=60
     Speed Alpha:NiagaraFloat=$System.Speed_Alpha
    )
    Use Spawn Probability=false
   ) {module=SpawnRate001}
  ]
  particleSpawnScript [
   &VC_PaintedVariantsOverride(
    Black=<rgba 0.0191 0.0191 0.0191 0.4>
    BurntSienna=<rgba 0.1 0.0324 0.02 0.6>
    Cobalt=<rgba 0.0296 0.0823 0.5457 0.6>
    Crimson=<rgba 0.205 0.0078 0.0078 0.25>
    DefaultColor=<rgba 0.35 0.35 0.7 0.5>
    ForestGreen=<rgba 0.0039 0.17 0.0039 0.5>
    Grey=<rgba 0.1 0.1 0.1 0.4>
    Lime=<rgba 0.0498 0.1 0 0.6>
    Orange=<rgba 0.3 0.0614 0.0062 0.5>
    Pink=<rgba 0.4 0.1174 0.3304 0.25>
    Purple=<rgba 0.1362 0.017 0.4 0.4>
    R1 Black:LinearColor=<rgba 0.3 0.3 0.3 1>
    R1 BurntSienna:LinearColor=<rgba 1 0.4481 0.3 1>
    R1 Cobalt:LinearColor=<rgba 1 1.1665 10 1>
    R1 Crimson:LinearColor=<rgba 1 0.1 0.1 1>
    R1 ForestGreen:LinearColor=<rgba 0 1 0.0505 1>
    R1 Grey:LinearColor=<rgba 0.8 0.8 0.8 1>
    R1 Lime:LinearColor=<rgba 0.4 0.8 0 1>
    R1 Orange:LinearColor=<rgba 5 1.744 0 1>
    R1 Pink:LinearColor=<rgba 2 1.6 1.8545 1>
    R1 Purple:LinearColor=<rgba 0.7052 0.5 1 1>
    R1 SkyBlue:LinearColor=<rgba 0.1073 0.3929 0.5365 1>
    R2 DefaultColor:LinearColor=<rgba 0.4 0.4 1 1>
    Saffron=<rgba 0.3 0.2307 0 0.5>
    SkyBlue=<rgba 0 0.1996 0.3 0.3>
    TitaniumWhite=<rgba 0.3 0.3 0.3 0.3>
    Use R1 Painted Overrides=true
    Use R2 Painted Overrides=true
   )
   /SetVariables(Particles.TagsLocalsEmbersTriangles:NiagaraInt32=&SpawnGroup()) {isSetParameters=true}
   &InitializeParticle(
    Color=$Particles.Color
    Lifetime Max=/VehicleCosmetics/FX/Niagara/Modules/MergeEmitters/VC_FloatBySpawnGroup(
     SpawnGroup or ExecIndex:NiagaraInt32=0
     CustomInt:NiagaraInt32=0
     MaximumSpawnGroups:NiagaraInt32=2
     Float SG 0:NiagaraFloat=&MakeCustomFloatFromBool(Bool=$User.bIsFrontendPreview, True Float=0.6, False Float=0.6)
     Float SG 1:NiagaraFloat=&MakeCustomFloatFromBool(Bool=$User.bIsFrontendPreview, True Float=0.4, False Float=0.2)
     Float SG 2:NiagaraFloat=0
     Float SG 3:NiagaraFloat=0
     Float SG 4:NiagaraFloat=0
     Float SG 5:NiagaraFloat=0
    )
    Lifetime Min:NiagaraFloat=&VC_FloatBySpawnGroup(
     SpawnGroup or ExecIndex:NiagaraInt32=0
     CustomInt:NiagaraInt32=0
     MaximumSpawnGroups:NiagaraInt32=2
     Float SG 0:NiagaraFloat=&MakeCustomFloatFromBool(Bool=$User.bIsFrontendPreview, True Float=0.1, False Float=0.1)
     Float SG 1:NiagaraFloat=&MakeCustomFloatFromBool(Bool=$User.bIsFrontendPreview, True Float=0.5, False Float=0.25)
     Float SG 2:NiagaraFloat=0
     Float SG 3:NiagaraFloat=0
     Float SG 4:NiagaraFloat=0
     Float SG 5:NiagaraFloat=0
    )
    Lifetime Mode=::NewEnumerator1
    Sprite Rotation Mode=::NewEnumerator2
    Sprite Size Max=&SelectVector2DFromArray(
     Direct Array Index=&SpawnGroup()
     Vector 2D Selection Array=DataInterface() {propertyValues {FloatData [<1.2 2>, <54 54>]}}
    )
    Sprite Size Min=&SelectVector2DFromArray(
     Direct Array Index=&SpawnGroup()
     Vector 2D Selection Array=DataInterface() {propertyValues {FloatData [<0.8 1.2>, <48 48>]}}
    )
    Sprite Size Mode=::NewEnumerator2
    UsePositionOffset=false
    Write Material Random=true
   )
   &Color(Color=/VehicleCosmetics/FX/Niagara/Modules/MergeEmitters/VC_Dynamic_ColorBySpawnGroup(
     Color SG 0:LinearColor=&MakeCustomLinearColorFromBool(Bool=$User.bIsPainted, True Color=$User.PaintedColor, False Color=<rgba 0.4 0.4 1 1>)
     MaximumSpawnGroups:NiagaraInt32=2
     Color SG 1:LinearColor=&MakeCustomLinearColorFromBool(Bool=$User.bIsPainted, True Color=$User.PaintedColor, False Color=<rgba 0.4 0.4 1 1>)
     Color SG 2:LinearColor=<rgba 1 1 1 1>
     Color SG 3:LinearColor=<rgba 1 1 1 1>
     Color SG 4:LinearColor=<rgba 1 1 1 1>
    )) {enabled=false}
   &PartitionParticles()
   &SkeletalMeshLocation(
    Bone / Socket Index=&Add_Integer(A=$Particles.PartitionParticles.Partition, B=4)
    Bone / Socket Sampling Mode=::NewEnumerator5
    Interpolate Mesh Sampling=false
    MaskSpawnGroup=false
    Mesh Sampling Type=::NewEnumerator6
    Sampled Position Offset=/VehicleCosmetics/FX/Niagara/Modules/MergeEmitters/VC_VectorBySpawnGroup(
     SpawnGroup or ExecIndex:NiagaraInt32=0
     CustomInt:NiagaraInt32=0
     MaximumSpawnGroups:NiagaraInt32=2
     Vector Group 0:Vector3f=&AddVectorToPosition(Position=<-50 0 0>, Vector=$User.Offset)
     Vector Group 1:Vector3f=&AddVectorToPosition(Position=<-30 0 0>, Vector=$User.Offset)
     Vector Group 2:Vector3f=<0 0 0>
     Vector Group 3:Vector3f=<0 0 0>
     Vector Group 4:Vector3f=<0 0 0>
     Vector Group 5:Vector3f=<0 0 0>
    )
    Skeletal Mesh=$System.SkeletalMesh
   )
   &AddVelocity(Velocity=&VC_Mask_Vectors(
     FrontEnd Value:Vector3f=&MaskVectorBySpawnGroup(Masked Vector=&RandomRangeVector(Minimum=<-90 0 -50>, Maximum=<-100 0 50>), Passthrough Vector=<1 1 1>)
     blsFrontEndPreview:NiagaraBool=$User.bIsFrontendPreview
     SpeedMin Value:Vector3f=&MaskVectorBySpawnGroup(Masked Vector=&RandomRangeVector(Minimum=<-90 0 -50>, Maximum=<-100 0 50>), Passthrough Vector=<1 1 1>)
     SpeedMax Value:Vector3f=&MaskVectorBySpawnGroup(Masked Vector=&RandomRangeVector(Minimum=<-90 0 -50>, Maximum=<-100 1 50>), Passthrough Vector=<1 1 1>)
     Speed Alpha:NiagaraFloat=$System.Speed_Alpha
    ))
   &ShapeLocation(
    MaskSpawnGroup=false
    Override Seed=false
    Sphere Radius=&VC_FloatBySpawnGroup(
     SpawnGroup or ExecIndex:NiagaraInt32=0
     CustomInt:NiagaraInt32=0
     MaximumSpawnGroups:NiagaraInt32=2
     Float SG 0:NiagaraFloat=10
     Float SG 1:NiagaraFloat=20
     Float SG 2:NiagaraFloat=0
     Float SG 3:NiagaraFloat=0
     Float SG 4:NiagaraFloat=0
     Float SG 5:NiagaraFloat=0
    )
    UseOwnerScale=true
   )
  ]
  particleUpdateScript [
   &FN_DisableTireFX(
    Mask Index Attribute:NiagaraInt32=$Particles.PartitionParticles.Partition
    Mask Index FL=2
    Mask Index FR=3
    Mask Index RL=0
    Mask Index RR=1
   )
   &ParticleState()
   &SpriteRotationRate(Rotation Rate=&VC_FloatBySpawnGroup(
     SpawnGroup or ExecIndex:NiagaraInt32=0
     CustomInt:NiagaraInt32=0
     MaximumSpawnGroups:NiagaraInt32=2
     Float SG 0:NiagaraFloat=0
     Float SG 1:NiagaraFloat=&RandomRangeFloat(Minimum=0, Maximum=0.3)
     Float SG 2:NiagaraFloat=0
     Float SG 3:NiagaraFloat=0
     Float SG 4:NiagaraFloat=0
     Float SG 5:NiagaraFloat=0
    ))
   &ScaleColor(
    Scale Alpha=&VC_FloatBySpawnGroup(
     SpawnGroup or ExecIndex:NiagaraInt32=0
     CustomInt:NiagaraInt32=0
     MaximumSpawnGroups:NiagaraInt32=2
     Float SG 0:NiagaraFloat=1
     Float SG 1:NiagaraFloat=&FloatFromCurve(FloatCurve=DataInterface() {propertyValues {Curve {keys [<at 0.5 0.8>,<at 1 0>]}}})
     Float SG 2:NiagaraFloat=0
     Float SG 3:NiagaraFloat=0
     Float SG 4:NiagaraFloat=0
     Float SG 5:NiagaraFloat=0
    )
    Scale RGB=&VC_VectorBySpawnGroup(
     SpawnGroup or ExecIndex:NiagaraInt32=0
     CustomInt:NiagaraInt32=0
     MaximumSpawnGroups:NiagaraInt32=2
     Vector Group 0:Vector3f=&VectorFromFloat(Value=&FloatFromCurve(FloatCurve=DataInterface() {propertyValues {Curve {keys [<at 0 10>,<at 1 0>]}}}))
     Vector Group 1:Vector3f=<2 2 2>
     Vector Group 2:Vector3f=<0 0 0>
     Vector Group 3:Vector3f=<0 0 0>
     Vector Group 4:Vector3f=<0 0 0>
     Vector Group 5:Vector3f=<0 0 0>
    )
    ScaleA=true
    ScaleRGB=true
   ) {enabled=false}
   &ScaleSpriteSize(
    Scale Sprite Size Mode=::NewEnumerator0
    Uniform Scale Factor=&VC_FloatBySpawnGroup(
     SpawnGroup or ExecIndex:NiagaraInt32=0
     CustomInt:NiagaraInt32=0
     MaximumSpawnGroups:NiagaraInt32=2
     Float SG 0:NiagaraFloat=&FloatFromCurve(FloatCurve=DataInterface() {propertyValues {Curve {keys [{interpMode=RCIM_Cubic, tangentMode=RCTM_Break, time=0, value=7}, {interpMode=RCIM_Cubic, tangentMode=RCTM_Break, time=1, value=0}]}}})
     Float SG 1:NiagaraFloat=1
     Float SG 2:NiagaraFloat=0
     Float SG 3:NiagaraFloat=0
     Float SG 4:NiagaraFloat=0
     Float SG 5:NiagaraFloat=0
    )
   )
   &SolveForcesAndVelocity(Clamp Velocity=false, Limit Acceleration=false)
  ]
  rendererClasses [&NiagaraSpriteRendererProperties]
  renderers [
   {rendererClass=&NiagaraSpriteRendererProperties}
   {rendererClass=&NiagaraSpriteRendererProperties, rendererIndex=1}
   {rendererClass=&NiagaraSpriteRendererProperties, rendererIndex=2}
  ]
 }
 NE_VC_DriftTrail_WheelWrap_Parent {
  emitterUpdateScript [
   &EmitterState(Inactive Response=::NewEnumerator1, Life Cycle Mode=::NewEnumerator1, Loop Duration=1, UseLoopDelay=false)
   &SpawnBurst_Instantaneous(Spawn Count=2, Use Spawn Probability=false, UseLoopCountLimit=false)
  ]
  particleSpawnScript [
   &VC_PaintedVariantsOverride() {enabled=false}
   &InitializeParticle(
    Color=&MakeCustomLinearColorFromBool(Bool=$User.bIsPainted, True Color=$User.PaintedColor, False Color=<rgba 0.4 0.4 1 1>)
    Mesh Scale Mode=::NewEnumerator4
    Sprite Rotation Mode=::NewEnumerator2
    Sprite Size Mode=::NewEnumerator3
    UsePositionOffset=false
    Write Material Random=true
   )
   &PartitionParticles()
   &SkeletalMeshLocation(
    Bone / Socket Index=&Add_Integer(A=$Particles.PartitionParticles.Partition, B=0)
    Bone / Socket Sampling Mode=::NewEnumerator5
    Interpolate Mesh Sampling=false
    MaskSpawnGroup=false
    Mesh Sampling Type=::NewEnumerator6
    Skeletal Mesh=$System.SkeletalMesh
   )
  ]
  particleUpdateScript [
   &FN_DisableTireFX(
    Mask Index Attribute:NiagaraInt32=$Particles.PartitionParticles.Partition
    Mask Index FL=2
    Mask Index FR=3
    Mask Index RL=0
    Mask Index RR=1
   )
   &VC_WheelWrapScale(Wheel Index Attribute:NiagaraInt32=&Add_Integer(A=$Particles.PartitionParticles.Partition, B=2))
   &DynamicMaterialParameters(
    Index 0 Param 1=&Multiply_Float(
     A=30
     B=&MakeCustomFloatFromBool(Bool=$User.bIsPainted, True Float=$User.OutlinePainted, False Float=1)
    )
    Index 0 Param 2=1
    Index 0 Param 3=0.0022
    Index 0 Param 4=10
    Index 1 Param 1=&MakeCustomFloatFromBool(Bool=$User.bIsFrontendPreview, True Float=1.33, False Float=3)
    Index 1 Param 2=&MakeCustomFloatFromBool(Bool=$User.bIsFrontendPreview, True Float=1.7, False Float=3.4)
    Index 1 Param 3=&MakeCustomFloatFromBool(Bool=$User.bIsFrontendPreview, True Float=1.5, False Float=3)
    Module.Index0Param0Write=true
    Module.Index0Param1Write=true
    Module.Index0Param2Write=true
    Module.Index0Param3Write=true
    Module.Index1Param0Write=true
    Module.Index1Param1Write=true
    Module.Index1Param2Write=true
    Module.Index1Param3Write=true
    Write Parameter Index 1=true
   )
   &ScaleColor(Scale RGB=<9 9 9>, ScaleA=true, ScaleRGB=true) {module=ScaleColor001}
   &ParticleState(Kill Particles When Lifetime Has Elapsed=false, Loop Particles Lifetime=true)
   &UpdateMeshOrientation(Rotation Coordinate Space=::NewEnumerator2, Rotation Vector=<0 1 0>) {enabled=false}
   &UpdateMeshOrientation(Rotation Coordinate Space=::NewEnumerator2, Rotation Rate=-100, Rotation Vector=<0 0 0>) {enabled=false, module=UpdateMeshOrientation001}
   &SkeletalMeshLocation(
    Bone / Socket Index=&Add_Integer(A=$Particles.PartitionParticles.Partition, B=4)
    Bone / Socket Sampling Mode=::NewEnumerator5
    MaskSpawnGroup=false
    Mesh Sampling Type=::NewEnumerator6
    Skeletal Mesh=$System.SkeletalMesh
   ) {module=SkeletalMeshLocation001}
   &SolveForcesAndVelocity(Clamp Velocity=false, Limit Acceleration=false)
  ]
  rendererClasses [&NiagaraMeshRendererProperties]
  renderers [{rendererClass=&NiagaraMeshRendererProperties, rendererIndex=0}]
 }
 NE_VC_DriftTrail_WheelWrap_Parent001 {
  bEnabled=false
  emitterUpdateScript [
   &EmitterState(
    Inactive Response=::NewEnumerator1
    Life Cycle Mode=::NewEnumerator1
    Loop Behavior=::NewEnumerator1
    Loop Duration Mode=::NewEnumerator1
    UseLoopDelay=false
   )
   &SpawnBurst_Instantaneous(Spawn Count=2, Use Spawn Probability=false, UseLoopCountLimit=false)
  ]
  particleSpawnScript [
   &VC_PaintedVariantsOverride() {enabled=false}
   &InitializeParticle(
    Color Mode=::NewEnumerator0
    Lifetime=5
    Mesh Scale Mode=::NewEnumerator4
    Sprite Rotation Mode=::NewEnumerator2
    Sprite Size Mode=::NewEnumerator3
    UsePositionOffset=false
    Write Material Random=true
   )
   &PartitionParticles()
   &SkeletalMeshLocation(
    Bone / Socket Index=&Add_Integer(A=$Particles.PartitionParticles.Partition, B=4)
    Bone / Socket Sampling Mode=::NewEnumerator5
    Interpolate Mesh Sampling=false
    MaskSpawnGroup=false
    Mesh Sampling Type=::NewEnumerator6
    Skeletal Mesh=$System.SkeletalMesh
   )
   &InitialMeshOrientation(Mesh Orientation Mode=::NewEnumerator7, Rotation=<0 0 0.5>, Use Rotation Vector=true)
   &Color(Color=<rgba 0 1 0 1>)
  ]
  particleUpdateScript [
   &FN_DisableTireFX(
    Mask Index Attribute:NiagaraInt32=$Particles.PartitionParticles.Partition
    Mask Index FL=2
    Mask Index FR=3
    Mask Index RL=0
    Mask Index RR=1
   )
   &VC_WheelWrapScale(Wheel Index Attribute:NiagaraInt32=&Add_Integer(A=$Particles.PartitionParticles.Partition, B=2))
   &ParticleState(Kill Particles When Lifetime Has Elapsed=false)
   &UpdateMeshOrientation(Rotation Coordinate Space=::NewEnumerator2, Rotation Vector=<0 1 0>) {enabled=false}
   &SkeletalMeshLocation(
    Bone / Socket Index=&Add_Integer(A=$Particles.PartitionParticles.Partition, B=4)
    Bone / Socket Sampling Mode=::NewEnumerator5
    MaskSpawnGroup=false
    Mesh Sampling Type=::NewEnumerator6
    Skeletal Mesh=$System.SkeletalMesh
   ) {module=SkeletalMeshLocation001}
   &SolveForcesAndVelocity(Clamp Velocity=false, Limit Acceleration=false)
   &ScaleColor(
    Scale RGB=&VectorFromFloat(Value=&Waveform("Waveform [1]"=::NewEnumerator0, "[1] Phase Scale"=5, "[1] Frequency"=0.5, "[1] Amplitude Min/Max"=<50 100>))
    ScaleA=true
    ScaleRGB=true
   )
  ]
  rendererClasses [&NiagaraMeshRendererProperties]
  renderers [{rendererClass=&NiagaraMeshRendererProperties, rendererIndex=0}]
 }
 Trail_NV002 {
  emitterUpdateScript [
   &EmitterState()
   &VC_RateQualityHandler(Default Rate=250)
   &VC_SpawnRibbon_v1(Number of Ribbons=2, SpawnRate=$Emitter.Spawn Rate)
  ]
  particleSpawnScript [
   &VC_PaintedVariantsOverride() {enabled=false}
   &VC_InitializeRibbon_v1(
    Color=&MakeCustomLinearColorFromBool(Bool=$User.bIsPainted, True Color=$User.PaintedColor, False Color=<rgba 0.4 0.4 1 1>)
    Custom Offset=true
    Enable Skeletal Position=true
    Lifetime=&MakeCustomFloatFromBool(Bool=$User.bIsFrontendPreview, True Float=1, False Float=1)
    Local Vector Offset=&AddVectorToPosition(
     Position=&MakeCustomVectorFromBool(Bool=$User.bIsFrontendPreview, True Vector=<20 0 -2>, False Vector=<10 0 -2>)
     Vector=$User.Offset
    )
    Offset Amount=1
    Ribbon Width:NiagaraFloat=&MakeCustomFloatFromBool(
     Bool=$User.bIsFrontendPreview
     True Float=&Multiply_Float(A=6, B=6)
     False Float=&Multiply_Float(A=6, B=6)
    )
   )
   &AddVelocity(Velocity=&VC_Mask_Vectors(
     FrontEnd Value:Vector3f=<-850 0 0>
     blsFrontEndPreview:NiagaraBool=$User.bIsFrontendPreview
     SpeedMin Value:Vector3f=<-1000 0 0>
     SpeedMax Value:Vector3f=<-1000 0 0>
     Speed Alpha:NiagaraFloat=$System.Speed_Alpha
    )) {module=AddVelocity001}
  ]
  particleUpdateScript [
   &FN_DisableTireFX(Mask Index Attribute:NiagaraInt32=$Particles.SpawnGroup, Mask Index FL=2, Mask Index FR=3, Mask Index RL=0, Mask Index RR=1)
   &ScaleColor(
    Scale Alpha=&FloatFromCurve(
     FloatCurve=DataInterface() {propertyValues {Curve {keys [
         {interpMode=RCIM_Cubic, tangentMode=RCTM_Break, time=0, value=0}
         {interpMode=RCIM_Cubic, tangentMode=RCTM_Break, time=0.1, value=1}
         {interpMode=RCIM_Cubic, tangentMode=RCTM_Break, time=0.9, value=1}
         {interpMode=RCIM_Cubic, tangentMode=RCTM_Break, time=1, value=0}
        ]}}}
     Scale Curve=2
    )
    Scale RGB=&Multiply_VectorByFloat(
     Vector=&Multiply_VectorByFloat(
      Vector=&VectorFromFloat(Value=&FloatFromCurve(
        FloatCurve=DataInterface() {propertyValues {Curve {keys [{interpMode=RCIM_Cubic, tangentMode=RCTM_Break, time=0, value=4}, {interpMode=RCIM_Cubic, tangentMode=RCTM_Break, time=0.5, value=1.5}]}}}
        Scale Curve=1.5
       ))
      Float=1
     )
     Float=$User.PaintedScalar
    )
    ScaleA=true
    ScaleRGB=true
   ) {module=ScaleColor001}
   &ScaleRibbonWidth(Ribbon Width Scale=&FloatFromCurve(FloatCurve=DataInterface() {propertyValues {Curve {keys [
         {interpMode=RCIM_Cubic, leaveTangent=10.0537, tangentMode=RCTM_Break, time=0.07, value=0}
         {interpMode=RCIM_Cubic, tangentMode=RCTM_Break, time=0.1, value=1.4}
         <at 0.5 0.8>
         <at 0.9 0>
        ]}}}))
   &DynamicMaterialParameters(
    Index 0 Param 1=&FloatFromCurve(
     FloatCurve=DataInterface() {propertyValues {Curve {keys [<at 0 30>,<at 15 1>]}}}
     Scale Curve=&MakeCustomFloatFromBool(Bool=$User.bIsPainted, True Float=$User.OutlinePainted, False Float=1)
    )
    Index 0 Param 2=1
    Index 0 Param 3=0.0022
    Index 0 Param 4=10
    Index 1 Param 1=&MakeCustomFloatFromBool(Bool=$User.bIsFrontendPreview, True Float=1.33, False Float=3)
    Index 1 Param 2=&MakeCustomFloatFromBool(Bool=$User.bIsFrontendPreview, True Float=1.7, False Float=3.4)
    Index 1 Param 3=&MakeCustomFloatFromBool(Bool=$User.bIsFrontendPreview, True Float=1.5, False Float=3)
    Module.Index0Param0Write=true
    Module.Index0Param1Write=true
    Module.Index0Param2Write=true
    Module.Index0Param3Write=true
    Module.Index1Param0Write=true
    Module.Index1Param1Write=true
    Module.Index1Param2Write=true
    Module.Index1Param3Write=true
    Write Parameter Index 1=true
   )
   &ParticleState()
   &InheritSourceMovement(
    Applied Position Scale=&Multiply_VectorByFloat(
     Vector=<1 1 1>
     Float=&FloatFromCurve(FloatCurve=DataInterface() {propertyValues {Curve {keys [
          {interpMode=RCIM_Cubic, leaveTangent=-0.0051, tangentMode=RCTM_Break, time=0, value=1}
          {arriveTangent=-0.247, interpMode=RCIM_Cubic, tangentMode=RCTM_Break, time=1, value=0.9}
         ]}}})
    )
    ApplyForce=false
    ApplyPosition=true
    ApplyPositionFalloff=false
   )
   &SolveForcesAndVelocity(Clamp Velocity=false, Limit Acceleration=false)
   &VC_UpdateRibbon_v1(Enable Skeletal Position=true)
  ]
  rendererClasses [&NiagaraRibbonRendererProperties]
  renderers [{rendererClass=&NiagaraRibbonRendererProperties}]
 }
 Triangles_Trail {
  emitterUpdateScript [
   &EmitterState()
   &SpawnRate(
    SpawnRate=&VC_Mask_Floats(
     FrontEnd Value:NiagaraFloat=60
     blsFrontEndPreview:NiagaraBool=$User.bIsFrontendPreview
     SpeedMin Value:NiagaraFloat=60
     SpeedMax Value:NiagaraFloat=60
     Speed Alpha:NiagaraFloat=$System.Speed_Alpha
    )
    Use Spawn Probability=false
   )
  ]
  particleSpawnScript [
   &VC_PaintedVariantsOverride(
    Black=<rgba 0.0191 0.0191 0.0191 0.4>
    BurntSienna=<rgba 0.1 0.0324 0.02 0.6>
    Cobalt=<rgba 0.0296 0.0823 0.5457 0.6>
    Crimson=<rgba 0.205 0.0078 0.0078 0.25>
    DefaultColor=<rgba 0.35 0.35 0.7 0.5>
    ForestGreen=<rgba 0.0039 0.17 0.0039 0.5>
    Grey=<rgba 0.1 0.1 0.1 0.4>
    Lime=<rgba 0.0498 0.1 0 0.6>
    Orange=<rgba 0.3 0.0614 0.0062 0.5>
    Pink=<rgba 0.4 0.1174 0.3304 0.25>
    Purple=<rgba 0.1362 0.017 0.4 0.4>
    R1 Black:LinearColor=<rgba 0.3 0.3 0.3 1>
    R1 BurntSienna:LinearColor=<rgba 1 0.4481 0.3 1>
    R1 Cobalt:LinearColor=<rgba 1 1.1665 10 1>
    R1 Crimson:LinearColor=<rgba 1 0.1 0.1 1>
    R1 ForestGreen:LinearColor=<rgba 0 1 0.0505 1>
    R1 Grey:LinearColor=<rgba 0.8 0.8 0.8 1>
    R1 Lime:LinearColor=<rgba 0.4 0.8 0 1>
    R1 Orange:LinearColor=<rgba 5 1.744 0 1>
    R1 Pink:LinearColor=<rgba 2 1.6 1.8545 1>
    R1 Purple:LinearColor=<rgba 0.7052 0.5 1 1>
    R1 SkyBlue:LinearColor=<rgba 0.1073 0.3929 0.5365 1>
    Saffron=<rgba 0.3 0.2307 0 0.5>
    SkyBlue=<rgba 0 0.1996 0.3 0.3>
    TitaniumWhite=<rgba 0.3 0.3 0.3 0.3>
    Use R1 Painted Overrides=true
   )
   &InitializeParticle(
    Color=$Particles.Color
    Lifetime Max=0.25
    Lifetime Min=0.5
    Lifetime Mode=::NewEnumerator1
    Position Mode=::NewEnumerator0
    Sprite Rotation Mode=::NewEnumerator2
    Sprite Size Max=<54 54>
    Sprite Size Min=<48 48>
    Sprite Size Mode=::NewEnumerator2
    Write Material Random=true
   )
   &PartitionParticles()
   &SkeletalMeshLocation(
    Bone / Socket Index=&Add_Integer(A=$Particles.PartitionParticles.Partition, B=4)
    Bone / Socket Sampling Mode=::NewEnumerator5
    Interpolate Mesh Sampling=false
    MaskSpawnGroup=false
    Mesh Sampling Type=::NewEnumerator6
    Sampled Position Offset=&AddVectorToPosition(Position=<-30 0 0>, Vector=$User.Offset)
    Skeletal Mesh=$System.SkeletalMesh
   )
   &AddVelocity(Velocity=&VC_Mask_Vectors(
     FrontEnd Value:Vector3f=<-1000 0 0>
     blsFrontEndPreview:NiagaraBool=$User.bIsFrontendPreview
     SpeedMin Value:Vector3f=<-1000 0 0>
     SpeedMax Value:Vector3f=<-1000 0 0>
     Speed Alpha:NiagaraFloat=$System.Speed_Alpha
    ))
   &ShapeLocation(MaskSpawnGroup=false, Override Seed=false, Sphere Radius=20, UseOwnerScale=true)
  ]
  particleUpdateScript [
   &FN_DisableTireFX(
    Mask Index Attribute:NiagaraInt32=$Particles.PartitionParticles.Partition
    Mask Index FL=2
    Mask Index FR=3
    Mask Index RL=0
    Mask Index RR=1
   )
   &ScaleColor(
    Scale Alpha=&FloatFromCurve(FloatCurve=DataInterface() {propertyValues {Curve {keys [<at 0.5 0.8>,<at 1 0>]}}})
    Scale RGB=&Multiply_VectorByFloat(
     Vector=<2 2 2>
     Float=&MakeCustomFloatFromBool(Bool=$User.bIsPainted, True Float=$User.ScaleColorPaintedTriang, False Float=1)
    )
    ScaleA=true
    ScaleRGB=true
   ) {enabled=false}
   &SubUVAnimation(
    "Sprite Renderer "=DataInterface() {propertyValues {SpriteRenderer=&self:Triangles_Trail_4.NiagaraSpriteRendererProperties_4}}
    SubUV Animation Mode:ENiagara_SubUVLookupModeV2=::NewEnumerator1
    SubUV Curve=DataInterface() {propertyValues {Curve {keys [{arriveTangent=4.1785, time=0, value=0}, {leaveTangent=0.2086, time=1, value=1}]}}}
    UseEndFrame=false
    UseStartFrame=false
   )
   &ParticleState()
   &SpriteRotationRate(Rotation Rate=&RandomRangeFloat(Minimum=0, Maximum=0.3))
   &Drag(Drag=0.5, Use Linear Drag=true, Use Rotational Drag=true)
   &InheritSourceMovement(
    Applied Position Scale=&Multiply_VectorByFloat(
     Vector=<1 1 1>
     Float=&FloatFromCurve(FloatCurve=DataInterface() {propertyValues {Curve {keys [
          {interpMode=RCIM_Cubic, leaveTangent=-0.0051, tangentMode=RCTM_Break, time=0, value=1}
          {arriveTangent=-0.247, interpMode=RCIM_Cubic, tangentMode=RCTM_Break, time=1, value=0.9}
         ]}}})
    )
    ApplyForce=false
    ApplyPosition=true
    ApplyPositionFalloff=false
   ) {module=InheritSourceMovement002}
   &SolveForcesAndVelocity(Clamp Velocity=false, Limit Acceleration=false)
   &DynamicMaterialParameters(
    Index 0 Param 1=&MakeCustomFloatFromBool(Bool=$User.bIsPainted, True Float=$User.TriOutline, False Float=3)
    Module.Index0Param0Write=true
    Module.Index0Param1Write=true
    Module.Index0Param2Write=true
    Module.Index0Param3Write=true
   )
  ]
  rendererClasses [&NiagaraSpriteRendererProperties]
  renderers [
   {rendererClass=&NiagaraSpriteRendererProperties}
   {rendererClass=&NiagaraSpriteRendererProperties, rendererIndex=1}
  ]
 }
}
summary {
 systemName=DriftTrail_CyberSecurity_MP
 emitters [
  {emitterName=NE_VC_DriftTrail_WheelWrap_Parent, rendererClasses [&NiagaraMeshRendererProperties]}
  {emitterName=Embers_Triangles, rendererClasses [&NiagaraSpriteRendererProperties]}
  {emitterName=Triangles_Trail, rendererClasses [&NiagaraSpriteRendererProperties]}
  {emitterName=Trail_NV002, rendererClasses [&NiagaraRibbonRendererProperties]}
  {bEnabled=false, emitterName=NE_VC_DriftTrail_WheelWrap_Parent001, rendererClasses [&NiagaraMeshRendererProperties]}
 ]
 userVariables [
  User.bIsFrontendPreview {type=NiagaraBool, defaultValue=true}
  User.bIsPainted {type=NiagaraBool}
  User.BorderPowerColor {type=NiagaraFloat, defaultValue=1, description=""}
  User.ColorVariantID {type=NiagaraInt32}
  User.NoiseInnerColor {type=NiagaraFloat, defaultValue=1, description=""}
  User.Offset {type=Vector3f, description="", defaultValue {struct=Vector3f, value=<0 0 0>}}
  User.OutlinePainted {type=NiagaraFloat, defaultValue=1, description=""}
  User.PaintedColor {type=LinearColor, description="", defaultValue {struct=LinearColor, value=<rgba 1 1 1 1>}}
  User.PaintedScalar {type=NiagaraFloat, defaultValue=1, description=""}
  User.PlaneColorPower {type=NiagaraFloat, defaultValue=1, description=""}
  User.ScalarPainted {type=NiagaraFloat, defaultValue=0.5, description=""}
  User.ScaleColorPaintedTriang {type=NiagaraFloat, defaultValue=1, description=""}
  User.TriOutline {type=NiagaraFloat, defaultValue=3, description=""}
  User.VehicleFlags {type=NiagaraInt32, defaultValue=806289408}
  User.Wheel_Radius_Array {
   type=NiagaraDataInterfaceArrayFloat
   defaultValue {
    struct=NiagaraExt_VariableValue_DataInterface
    value {
     dataInterface=&self:NiagaraDataInterfaceArrayFloat_0
     dataInterfaceClass=&NiagaraDataInterfaceArrayFloat
    }
   }
  }
  User.Wheel_Width_Array {
   type=NiagaraDataInterfaceArrayFloat
   defaultValue {
    struct=NiagaraExt_VariableValue_DataInterface
    value {
     dataInterface=&self:NiagaraDataInterfaceArrayFloat_1
     dataInterfaceClass=&NiagaraDataInterfaceArrayFloat
    }
   }
  }
 ]
}
