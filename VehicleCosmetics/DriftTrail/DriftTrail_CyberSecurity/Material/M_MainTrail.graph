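# Translucent unlit material for the CyberSecurity drift trail cosmetic, usable on Niagara ribbons and mesh particles.
# Time-panned cyber mask, inner-line and cell-noise textures supply the channels for the pattern.
# Three sine pulses with dynamic-parameter speeds flicker layers tinted by particle color, which sum with outline and glow terms into an emissive adjustment function.
# A wheelwrap mask multiplied by a camera depth fade forms a separate masking term.
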Material /VehicleCosmetics/DriftTrail/DriftTrail_CyberSecurity/Material/M_MainTrail
graph {
 properties {
  bUsedWithNiagaraMeshParticles=true
  bUsedWithNiagaraRibbons=true
  blendMode=BLEND_Translucent
  pixelDepthOffsetMode=PDOM_Legacy
  shadingModel=MSM_Unlit
  usedShadingModels=Unlit
 }
 nodes [
  n1 = Multiply(A=Multiply(A=NamedRerouteUsage(), B=DynamicParameter() {paramNames [Outline,InnerLines,PowerMask,PowerGlow]}.Outline), B=NamedRerouteUsage())
  MaterialExpressionVertexColor_1 = ParticleColor()
  n2 = NamedRerouteUsage()
  n3 = DynamicParameter() {parameterIndex=1, defaultValue=<rgba 1.33 1.7 1.5 0>, paramNames [TimeValue_1,TimeValue_2,TimeValue_3,0]}
  MaterialExpressionConstantClamp_0 = Saturate(None=Sine(None=MaterialFunctionCall(Speed=@n3.TimeValue_1) {materialFunction=&TimeWithSpeedVariable}.Time))
  n4 = LinearInterpolate(A=@n2, B=Multiply(A=@n2) {constB=0.35}, Alpha=@MaterialExpressionConstantClamp_0)
  n5 = NamedRerouteUsage()
  MaterialExpressionConstantClamp_1 = Saturate(None=Sine(None=MaterialFunctionCall(Speed=@n3.TimeValue_2) {materialFunction=&TimeWithSpeedVariable}.Time) {period=2})
  n6 = NamedRerouteUsage()
  MaterialExpressionConstantClamp_2 = Saturate(None=Sine(None=MaterialFunctionCall(Speed=@n3.TimeValue_3) {materialFunction=&TimeWithSpeedVariable}.Time) {period=3})
  n7 = Add(
   A=LinearInterpolate(A=@n5, B=Multiply(A=@n5) {constB=0.35}, Alpha=@MaterialExpressionConstantClamp_1)
   B=LinearInterpolate(A=@n6, B=Multiply(A=@n6) {constB=0.35}, Alpha=@MaterialExpressionConstantClamp_2)
  )
  n8 = Multiply(
   A=Multiply(A=@MaterialExpressionVertexColor_1.A, B=Add(A=@n4, B=@n7))
   B=NamedRerouteDeclaration(None=@MaterialExpressionVertexColor_1.RGB) {name=Particles RGB, nodeColor=<rgba 0.875 0 1 1>}
  )
  n9 = Multiply(A=NamedRerouteUsage(), B=Constant() {r=8})
  n10 = Multiply(A=NamedRerouteUsage(), B=NamedRerouteUsage())
  MaterialFunctionCall(Emissive=Add(A=Multiply(A=@n9, B=@n10), B=Add(A=@n1, B=@n8))) {materialFunction=&MF_FX_EmissiveAdjustments}
  n11 = Multiply(A=Time(), B=ScalarParameter() {defaultValue=1, parameterName=TimePanning})
  n12 = Panner(Coordinate=TextureCoordinate() {uTiling=5}, Time=@n11) {speedX=-2.3}
  n13 = TextureSample(UVs=@n12) {texture=/DealershipSharedLibrary/FX/Textures/TexturePackage/T_VCVFX_CyberMask_RGB}
  NamedRerouteDeclaration(None=@n13.B) {name=Mask B T3, nodeColor=<rgba 0 1 0.4844 1>}
  NamedRerouteDeclaration(None=@n13.G) {name=Mask G T3, nodeColor=<rgba 1 0.1641 0 1>}
  n14 = TextureSample(UVs=@n12) {texture=/DealershipSharedLibrary/FX/Textures/TexturePackage/T_VCVFX_CyberWhite_RGB}
  NamedRerouteDeclaration(None=@n14.R) {name=TEX4 R, nodeColor=<rgba 1 0 0.7969 1>}
  NamedRerouteDeclaration(None=@n14.G) {name=TEX4 G, nodeColor=<rgba 0 1 0.8828 1>}
  NamedRerouteDeclaration(None=@n14.B) {name=TEX4 B, nodeColor=<rgba 1 0.5625 0 1>}
  NamedRerouteDeclaration(None=TextureSample(UVs=@n12) {texture=/DealershipSharedLibrary/FX/Textures/TexturePackage/T_VCVFX_InnerLines_RGB}.B) {name=Mask B TEX2, nodeColor=<rgba 1 0.8293 0.1048 1>}
  n15 = TextureSample(UVs=Panner(Coordinate=TextureCoordinate() {uTiling=1.5, vTiling=1.5}, Time=@n11) {speedX=-1, speedY=0.2}) {texture=/DealershipSharedLibrary/FX/Textures/Noise/T_VCVFX_CellNoise2_A}
  NamedRerouteDeclaration(None=@n15.R) {name=TEX1, nodeColor=<rgba 1 0.2344 0 1>}
  Reroute(None=@n13.B)
  Reroute(None=@n13.G)
  Reroute(None=@MaterialExpressionVertexColor_1.RGB)
  Reroute(None=@n15.R)
  n16 = Constant() {r=0.98}
  n17 = NamedRerouteUsage()
  n18 = Constant() {r=100}
  n19 = Constant() {r=50}
  MaterialExpressionConstantClamp_7 = Saturate(None=Multiply(
    A=Multiply(A=@n17) {constB=5}
    B=MaterialFunctionCall(Fade Length=@n18, Fade Offset=@n19) {materialFunction=&CameraDepthFade}.Result
   ))
  Reroute(None=Multiply(
    A=MaterialFunctionCall!(Multiply, Power EXP=@n16) {materialFunction=/DealershipSharedLibrary/FX/Materials/MaterialFunctions/MF_VCVFX_WheelwrapMask}.Result
    B=@MaterialExpressionConstantClamp_7
   ))
  Reroute(None=@n12)
  Reroute(None=@n14.RGB)
  Reroute(None=@n14.R)
  Reroute(None=@n14.G)
  Reroute(None=@n14.B)
 ]
}
